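; Decorative double paper lantern string building prop that carries its material and tint over to its secondary light-source lantern mesh.
; The construction script runs the parent logic, then checks the building texture data for a valid first entry.
; When present, it assigns the main mesh's dynamic material to the lantern mesh and passes that entry's tint color to SetColoronSecondaryMesh.
; BeginPlay, overlap and Tick events are empty.
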(event EventBeginPlay)

(event Collision|EventActorBeginOverlap
  (params (OtherActor :type Actor%20Object%20Reference)))

(event EventTick
  (params (DeltaSeconds :type Float%20%28single-precision%29)))

(fn ConstructionScript
  (bind _staticmeshcomponent (Variables|Debug|GetStaticMeshComponent))
  (bind _material_instance_dynamic (Class|BTechArtGlobalLibrary|GetDynamicMaterialInstance _staticmeshcomponent))
  (|CallParentFunction)
  (bind _outtexturedataarray (Building|GetBuildingTextureData))
  (if (> (Utilities|Array|Length _outtexturedataarray) 0)
    (bind _outtexturedataarray_2 (Building|GetBuildingTextureData))
    (if (Utilities|IsValid (Utilities|Array|Get(acopy) _outtexturedataarray_2) :!class /Script/Engine.KismetSystemLibrary)
      (Rendering|Material|SetMaterial (Variables|Default|GetSM_LightSource_PaperLanternString_A) 0 (Class|BTechArtGlobalLibrary|GetDynamicMaterialInstance (Variables|Debug|GetStaticMeshComponent)))
      (Rendering|Material|SetMaterial (Variables|Default|GetSM_LightSource_PaperLanternString_A) 0 _material_instance_dynamic)
      (Rendering|Material|GetVectorParameterValue _material_instance_dynamic "L1 Tint (Expand)" :!class /Script/Engine.MaterialInstanceDynamic)
      (Default|SetColoronSecondaryMesh (Math|Conversions|ToLinearColor(Color) (Variables|/Script/FortniteGame.BuildingTextureData|GetTintColor (Utilities|Array|Get(acopy) _outtexturedataarray_2))) (Variables|Default|GetSM_LightSource_PaperLanternString_A)))))
