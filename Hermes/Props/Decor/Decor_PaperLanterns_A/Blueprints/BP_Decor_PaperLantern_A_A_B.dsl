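; Paper lantern decor building actor built on an adjustable light parent, with a light fixture, down rod and swinging anchor knot.
; Construction gives the fixture a dynamic material and tint from building texture data, and swaps the hidden base mesh for an added copy to bypass BuildingActor material overrides.
; It adds the knot at a fixture socket with world-position-offset swing material.
; Mesh and material references are registered through the exposed asset provider.
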(event EventBeginPlay
  (|CallParentFunction) ; This node is disabled and will not be called.\nDrag off pins to build functionality.
)

(event Collision|EventActorBeginOverlap
  (params (OtherActor :type Actor%20Object%20Reference))
  (|CallParentFunction OtherActor) ; This node is disabled and will not be called.\nDrag off pins to build functionality.
)

(event EventTick
  (params (DeltaSeconds :type Float%20%28single-precision%29))
  (|CallParentFunction DeltaSeconds) ; This node is disabled and will not be called.\nDrag off pins to build functionality.
)

(fn ApplyCorrectMIDFromOverride
  (bind _lightfixture (Variables|AdjustableLight|GetLightFixture))
  (bind _outtexturedataarray (Building|GetBuildingTextureData))
  (if (> (Utilities|Array|Length _outtexturedataarray) 0)
    (Rendering|Material|CreateDynamicMaterialInstance _lightfixture 0 (Variables|/Script/FortniteGame.BuildingTextureData|GetOverrideMaterial (Utilities|Array|Get(acopy) _outtexturedataarray)) "None" :!class /Script/Engine.PrimitiveComponent)
    (Default|SetColoronSecondaryMesh (Math|Conversions|ToLinearColor(Color) (Variables|/Script/FortniteGame.BuildingTextureData|GetTintColor (Utilities|Array|Get(acopy) _outtexturedataarray))) _lightfixture)))

(fn BuildExposedAssetProviderVersion(Implementation)
  (params (Version :type Fort%20Exposed%20Asset%20Provider%20Version%20Structure%20%28by%20ref%29))
  (|CallParentFunction Version)
  (ExposedAssetProvider|AddVersion Version 1)
  (return))

(fn CollectAssetReferencestoAllow(Implementation)
  (params (Collector :type Fort%20Exposed%20Asset%20Reference%20Collector%20Structure%20%28by%20ref%29))
  (|CallParentFunction Collector)
  (ExposedAssetProvider|CollectObjectPropertyAssetReferences Collector 0 (Variables|Default|GetSwingingMaterial))
  (ExposedAssetProvider|CollectObjectPropertyAssetReferences Collector 0 (Variables|Default|GetAnchorKnot))
  (ExposedAssetProvider|CollectObjectPropertyAssetReferences Collector 0 (Variables|Default|GetMainStaticMesh))
  (return))

(fn ConstructionScript
  (|CallParentFunction)
  (Rendering|Material|SetMaterial (Variables|AdjustableLight|GetLightFixture) 0 (Variables|Default|GetLightFixtureMaterial))
  (CallFunction|ApplyCorrectMIDFromOverride)
  (Rendering|SetVisibility (Variables|Debug|GetStaticMeshComponent) false false)
  ; Ugly workaround for buildingactor forcibly overriding attempts at assigning a unique MID not driven by TextureData asset
  (bind _returnvalue (AddComponent|NODE_AddStaticMeshComponent-0))
  (Components|StaticMesh|SetStaticMesh _returnvalue (Variables|/Script/Engine.StaticMeshComponent|GetStaticMesh (Variables|Debug|GetStaticMeshComponent)))
  ; Moving knot mesh to tail end of construction instead of attaching via component hierarchy
  (bind _returnvalue_2 (AddComponent|NODE_AddStaticMeshComponent-1 :RelativeTransform (Math|Transform|MakeTransform (.location (Transformation|GetSocketTransform (Variables|AdjustableLight|GetLightFixture) "None" "RTS_Actor")) "0.000000,-90.000000,0.000000" (.scale (Transformation|GetSocketTransform (Variables|AdjustableLight|GetLightFixture) "None" "RTS_Actor")))))
  (Components|StaticMesh|SetStaticMesh _returnvalue_2 (Variables|Default|GetAnchorKnot))
  ; For dem swings
  (Rendering|Material|SetMaterial _returnvalue_2 0 (Variables|Default|GetSwingingMaterial))
  (Rendering|LOD|SetEvaluateWorldPositionOffset _returnvalue_2 true :!class /Script/Engine.StaticMeshComponent)
  (Rendering|LOD|SetWorldPositionOffsetDisableDistance _returnvalue_2 5000 :!class /Script/Engine.StaticMeshComponent)
  (Rendering|Material|SetDefaultCustomPrimitiveDataFloat (Variables|AdjustableLight|GetDownRod) 3 1.0))
